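# Atlas_FNC is a material function, stored in a muscle-test folder for the large male player body M_LRG_Rumble, that remaps UVs into one cell of a texture atlas.
# It divides the texture coordinates by a scalar Scale, defaulting to a 4x4 grid, and adds the two-component Offset divided by the same Scale.
# Offset is therefore expressed in cell units, and the single output is the adjusted UV.
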MaterialFunction /Game/Characters/Player/Male/Large/Bodies/M_LRG_Rumble/Materials/Muscle_Test/Materials/Atlas_FNC
graph {nodes [
  n1 = TextureCoordinate()
  n2 = FunctionInput() {bUsePreviewValueAsDefault=true, inputName=Scale, inputType=FunctionInput_Scalar, previewValue=<4 0 0 1>}
  n3 = FunctionInput() {bUsePreviewValueAsDefault=true, inputName=Offset, inputType=FunctionInput_Vector2}
  FunctionOutput(None=Add(A=Divide(A=@n1, B=@n2), B=Divide(A=@n3, B=@n2)))
 ]}
